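# Niagara system for the DrapeUltra emote, named for its landing, that bursts a small puff of pale smoke sprites from a flat shape.
# One Smoke emitter partitions particles into spawn groups, which select lifetime, launch speed and alpha.
# Strong drag slows the sprites as they grow, follow source movement and fade in and out.
# Age-driven dynamic material parameters vary the look across two sprite renderers, and alpha passes through a front-end/gameplay modifier.
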NiagaraSystem /BRCosmetics/Animation/Game/MainPlayer/Emotes/DrapeUltra/FX/NS_Emote_DrapeUltra_Land
dependencies {
 usedDataInterfaces [&NiagaraDataInterfaceCurve]
 usedDynamicInputs [&RandomRangeVector, &Vector2DFromFloat]
 usedModules [&SystemState]
}
emitters {Smoke {
  emitterUpdateScript [
   &EmitterState(Life Cycle Mode=::NewEnumerator1, Loop Behavior=::NewEnumerator1, Loop Duration=0.2, UseLoopDelay=false)
   &SpawnBurst_Instantaneous(Spawn Count=25, Use Spawn Probability=false, UseLoopCountLimit=false)
  ]
  particleSpawnScript [
   /SetVariables(Particles.SpawnGroup=$Emitter.SpawnGroup) {isSetParameters=true}
   &PartitionParticles(Number Of Partitions=4)
   &InitializeParticle(
    Color Channel Mode:ENiagaraChannelCorrelation=::NewEnumerator0
    Color Maximum=<rgba 0.35 0.3477 0.336 0.25>
    Color Minimum=<rgba 0.35 0.3404 0.3251 0.2>
    Color Mode=::NewEnumerator2
    Lifetime=&MergeHelper_FloatSwitch(
     0=&RandomRangeFloat(Minimum=0.5, Maximum=0.8)
     1=&RandomRangeFloat(Minimum=0.75, Maximum=1.2)
     2=&RandomRangeFloat(Minimum=0.5, Maximum=0.5)
     3=?
     Index=$Particles.SpawnGroup
     Num=::NewEnumerator1
    )
    Sprite Rotation Mode=::NewEnumerator2
    Sprite Size Mode=::NewEnumerator1
    Uniform Sprite Size Max=35
    Uniform Sprite Size Min=20
    UsePositionOffset=false
    Write Material Random=true
   ) {module=InitializeParticle001}
   &ShapeLocation(
    Cylinder Height=0
    Cylinder Height Midpoint=0
    Cylinder Radius=10
    MaskSpawnGroup=false
    Offset Coordinate Space=::Local
    Offset Mode=::NewEnumerator0
    Override Seed=false
    Shape Primitive=::NewEnumerator1
    Surface Only=false
    UseOwnerScale=true
   )
   &AddVelocity(
    Module.Use Falloff Distance=false
    Override Seed=false
    Velocity Mode=::NewEnumerator1
    Velocity Speed=&MergeHelper_FloatSwitch(
     0=&RandomRangeFloat(Minimum=150, Maximum=500)
     1=&RandomRangeFloat(Minimum=350, Maximum=700)
     2=&RandomRangeFloat(Minimum=300, Maximum=500)
     3=?
     Index=$Particles.SpawnGroup
     Num=::NewEnumerator1
    )
   )
   &ApplyInitialForces()
  ]
  particleUpdateScript [
   &ParticleState()
   &Drag(Drag=7, Use Linear Drag=true, Use Rotational Drag=true)
   &InheritSourceMovement(ApplyForce=false, ApplyPosition=true, ApplyPositionFalloff=false)
   &SolveForcesAndVelocity(Clamp Velocity=false, Limit Acceleration=false)
   &ScaleSpriteSize(Uniform Curve Sprite Scale=DataInterface() {propertyValues {Curve {keys [
        {arriveTangent=1.491, interpMode=RCIM_Cubic, leaveTangent=1.491, tangentMode=RCTM_User, time=0, value=0.5}
        {arriveTangent=0.5, interpMode=RCIM_Cubic, leaveTangent=0.5, time=0.95, value=3}
        <at 1 1 RCIM_Cubic>
       ]}}})
   &DynamicMaterialParameters(
    Index 0 Param 1=&Add_Float(
     A=&RandomRangeFloat(Minimum=0, Maximum=1)
     B=&FloatFromCurve(
      FloatCurve=DataInterface() {propertyValues {Curve {keys [
          <at 0 0 RCIM_Cubic>
          {arriveTangent=1.5947, interpMode=RCIM_Cubic, leaveTangent=1.5947, tangentMode=RCTM_User, time=0.2557, value=0.5257}
          <at 1 1 RCIM_Cubic>
         ]}}}
      CurveIndex=$Particles.Age
      Scale Curve=0.25
     )
    )
    Index 0 Param 3=1
    Index 0 Param 4=1
    Index 2 Param 3=2
    Index 3 Param 1=1
    Index 3 Param 2=70
    Module.Param0WriteEnabled=true
    Module.Param0WriteEnabled1=true
    Module.Param0WriteEnabled2=true
    Module.Param0WriteEnabled3=true
    Module.Param1WriteEnabled=true
    Module.Param1WriteEnabled1=true
    Module.Param1WriteEnabled2=true
    Module.Param1WriteEnabled3=true
    Module.Param2WriteEnabled=true
    Module.Param2WriteEnabled1=true
    Module.Param2WriteEnabled2=true
    Module.Param2WriteEnabled3=true
    Module.Param3WriteEnabled=true
    Module.Param3WriteEnabled1=true
    Module.Param3WriteEnabled2=true
    Module.Param3WriteEnabled3=true
    Write Parameter Index 1=true
    Write Parameter Index 2=true
    Write Parameter Index 3=true
   ) {module=DynamicMaterialParameters001}
   &CameraOffset(Camera Offset Amount=5)
   &ScaleColor(
    Scale Alpha=&Multiply_Float(
     A=&FloatFromCurve(
      FloatCurve=DataInterface() {propertyValues {Curve {keys [{interpMode=RCIM_Cubic, tangentMode=RCTM_User, time=0, value=0}, <at 0.25 1 RCIM_Cubic>, <at 0.9 0>]}}}
      Scale Curve=&FN_FrontEndFloatModifier(Gameplay=0.2, Front End=0.2)
     )
     B=&MergeHelper_FloatSwitch(0=0.5, 1=1, 2=0.5, 3=?, Index=$Particles.SpawnGroup, Num=::NewEnumerator1)
    )
    Scale RGB=<5 5 5>
    ScaleA=true
    ScaleRGB=true
   )
  ]
  rendererClasses [&NiagaraSpriteRendererProperties]
  renderers [
   {rendererClass=&NiagaraSpriteRendererProperties}
   {rendererClass=&NiagaraSpriteRendererProperties, rendererIndex=1}
  ]
 }}
summary {
 systemName=NS_Emote_DrapeUltra_Land
 emitters [{emitterName=Smoke, rendererClasses [&NiagaraSpriteRendererProperties]}]
 userVariables [
  User.bIsFrontEnd {type=NiagaraBool}
  User.bIsFrontEndPreview {type=NiagaraBool}
 ]
}
