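# Niagara effect for the leap in the DrapeUltra player emote, a short burst of upward air and dust.
# AirLines streak sprites and a brief Air_Trail_Up ribbon shoot upward, while a Shockwave mesh expands and Smoke sprites puff out with per-spawn-group lifetime, speed and alpha.
# All emitters inherit source movement; air and shockwave colors follow time of day.
# A front-end modifier lowers trail, shockwave and smoke opacity outside gameplay.
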NiagaraSystem /BRCosmetics/Animation/Game/MainPlayer/Emotes/DrapeUltra/FX/NS_Emote_DrapeUltra_Leap
dependencies {
 usedDataInterfaces [&NiagaraDataInterfaceCurve, &NiagaraDataInterfaceVectorCurve, &NiagaraDataInterfaceColorCurve]
 usedDynamicInputs [&Vector2DFromFloat]
 usedModules [&SystemState]
}
emitters {
 AirLines {
  emitterUpdateScript [
   &EmitterState()
   &SpawnBurst_Instantaneous(Spawn Count=15, Use Spawn Probability=false, UseLoopCountLimit=false)
  ]
  particleSpawnScript [
   &InitializeParticle(
    Color=&MultiplyLinearColor(
     Linear Color=&RandomRangeLinearColor(Minimum=<rgba 1 0.7058 0.7 0.2>, Maximum=<rgba 1 0.8205 0.7 0.6>)
     Scale Factor=&MakeVector4_FromLinearColor(LinearColor=&FNTimeOfDayColor(
       Night Color=<rgba 0.1667 0.2211 0.8333 1>
       Morning Color=<rgba 0.6242 0.7654 0.9219 1>
       Day Color=<rgba 1 0.7073 0.5207 1>
       Evening Color=<rgba 0.6242 0.7654 0.9219 1>
      ))
    )
    Lifetime=&RandomRangeFloat(Minimum=0.15, Maximum=0.4)
    Sprite Size=<1 60>
    Sprite Size Mode=::NewEnumerator4
    UsePositionOffset=false
    Write Material Random=true
   )
   &AddVelocity(Velocity=&RandomRangeVector(Minimum=<0 0 200>, Maximum=<0 0 400>), Velocity Speed Scale=0.8)
   &ShapeLocation(
    Cylinder Radius=25
    MaskSpawnGroup=false
    Offset=<-10 0 30>
    Offset Coordinate Space=::Local
    Offset Mode=::NewEnumerator0
    Override Seed=false
    Shape Primitive=::NewEnumerator1
    Surface Only=false
    UseOwnerScale=true
   )
  ]
  particleUpdateScript [
   &ParticleState()
   &Drag(Use Linear Drag=true, Use Rotational Drag=true)
   &InheritSourceMovement(ApplyForce=false, ApplyPosition=true, ApplyPositionFalloff=false)
   &SolveForcesAndVelocity(Clamp Velocity=false, Limit Acceleration=false)
   &ScaleColor(
    Scale Alpha=&FloatFromCurve(FloatCurve=DataInterface() {propertyValues {Curve {keys [<at 0 1>,<at 0.4 0.85>,<at 0.6 0.7>,<at 1 0>]}}}, Scale Curve=0.75)
    ScaleA=true
    ScaleRGB=true
   )
  ]
  rendererClasses [&NiagaraSpriteRendererProperties]
  renderers [{rendererClass=&NiagaraSpriteRendererProperties}]
 }
 Air_Trail_Up {
  emitterUpdateScript [
   &EmitterState(Life Cycle Mode=::NewEnumerator1, Loop Behavior=::NewEnumerator1, Loop Duration=0.45, UseLoopDelay=false)
   &SpawnRate(SpawnRate=400, Use Spawn Probability=false)
  ]
  particleSpawnScript [
   &InitializeParticle(
    Color=&MultiplyLinearColor(
     Linear Color=&FNTimeOfDayColor(
      Night Color=<rgba 0.2908 0.4826 0.8333 1>
      Morning Color=<rgba 0.9219 0.6986 0.6242 1>
      Day Color=<rgba 0.8 0.6533 0.64 1>
      Evening Color=<rgba 0.9219 0.6242 0.649 1>
     )
     Scale Factor=&Vector4FromCurve(
      CurveIndex=$Emitter.NormalizedLoopAge
      Vector4Curve=DataInterface() {propertyValues {
        AlphaCurve {keys [<at 0 0>,<at 0.1526 0.8>,<at 1 0>]}
        BlueCurve {keys [<at 0 1>,<at 1 1>]}
        GreenCurve {keys [<at 0 1>,<at 1 1>]}
        RedCurve {keys [<at 0 1>,<at 1 1>]}
       }}
     )
    )
    Lifetime=0.1
    Ribbon Width=&FloatFromCurve(
     FloatCurve=DataInterface() {propertyValues {Curve {keys [<at 0 0.5>,<at 0.1 1>,<at 1 3>]}}}
     CurveIndex=$Emitter.NormalizedLoopAge
     Scale Curve=30
    )
    Ribbon Width Mode=::NewEnumerator1
    UsePositionOffset=false
    Write Material Random=true
   )
   &AddVelocity(Velocity=<0 0 1500>)
  ]
  particleUpdateScript [
   &ParticleState()
   &InheritSourceMovement(ApplyForce=false, ApplyPosition=true, ApplyPositionFalloff=false)
   &SolveForcesAndVelocity(Clamp Velocity=false, Limit Acceleration=false)
   &ScaleColor(
    Scale Alpha=&FloatFromCurve(
     FloatCurve=DataInterface() {propertyValues {Curve {keys [
         {interpMode=RCIM_Cubic, leaveTangentWeight=1.4087, tangentMode=RCTM_User, time=0, value=1}
         {arriveTangent=-1.4286, interpMode=RCIM_Cubic, leaveTangent=-1.4286, time=0.3, value=1}
         {interpMode=RCIM_Cubic, leaveTangent=0.45, time=0.7, value=0}
        ]}}}
     Scale Curve=&FN_FrontEndFloatModifier(Gameplay=0.75, Front End=0.35)
    )
    Scale RGB=&MakeVectorFromLinearColorRGB(Color=&ColorFromCurve(ColorCurve=DataInterface() {propertyValues {
        AlphaCurve {keys []}
        BlueCurve {keys [
          <at 0 0.3802>
          {arriveTangent=-0.9804, interpMode=RCIM_Cubic, leaveTangent=-0.9804, time=0.0299, value=0.8646}
          <at 0.17 0.2135>
          <at 0.35 0.3125 RCIM_Cubic>
         ]}
        GreenCurve {keys [
          <at 0 0.3802>
          {arriveTangent=-0.9804, interpMode=RCIM_Cubic, leaveTangent=-0.9804, time=0.0299, value=0.8646}
          <at 0.17 0.2135>
          <at 0.35 0.3125 RCIM_Cubic>
         ]}
        RedCurve {keys [
          <at 0 0.3802>
          {arriveTangent=-0.9804, interpMode=RCIM_Cubic, leaveTangent=-0.9804, time=0.0299, value=0.8646}
          <at 0.17 0.2135>
          <at 0.35 0.3125 RCIM_Cubic>
         ]}
       }}))
    ScaleA=true
    ScaleRGB=true
   )
   &ScaleRibbonWidth(Ribbon Width Scale=&FloatFromCurve(FloatCurve=DataInterface() {propertyValues {Curve {keys [<at 0 1 RCIM_Cubic>,<at 1 0.5 RCIM_Cubic>]}}}))
   &DynamicMaterialParameters(
    Index 0 Param 1=&FloatFromCurve(FloatCurve=DataInterface() {propertyValues {Curve {keys [<at 0.1 1>,<at 1 0>]}}})
    Index 0 Param 2=&FloatFromCurve(FloatCurve=DataInterface() {propertyValues {Curve {keys [<at 0 0 RCIM_Cubic>,<at 0.1 1 RCIM_Cubic>,<at 0.5 0 RCIM_Cubic>]}}})
    Index 0 Param 3=-1
    Index 0 Param 4=1
    Module.Param0WriteEnabled=true
    Module.Param1WriteEnabled=true
    Module.Param2WriteEnabled=true
    Module.Param3WriteEnabled=true
   )
  ]
  rendererClasses [&NiagaraRibbonRendererProperties]
  renderers [{rendererClass=&NiagaraRibbonRendererProperties}]
 }
 Shockwave {
  emitterUpdateScript [
   &EmitterState(Life Cycle Mode=::NewEnumerator1, Loop Behavior=::NewEnumerator1, Loop Duration=0.1, UseLoopDelay=true)
   &SpawnBurst_Instantaneous(Use Spawn Probability=false, UseLoopCountLimit=false)
  ]
  particleSpawnScript [
   &InitializeParticle(
    Color=&MultiplyLinearColor(
     Linear Color=&FNTimeOfDayColor(
      Night Color=<rgba 0.2083 0.2594 0.8333 1>
      Morning Color=<rgba 0.9219 0.6986 0.6242 1>
      Day Color=<rgba 1 0.7073 0.5207 1>
      Evening Color=<rgba 0.9219 0.6242 0.649 1>
     )
     Scale Factor=&MakeVector4_FromLinearColor(LinearColor=<rgba 1 1 1 0.3>)
    )
    Mesh Scale Mode=::NewEnumerator3
    Mesh Uniform Scale=2
    Position Offset=<0 0 80>
    Position Offset Coordinate Space=::Local
    UsePositionOffset=true
    Write Material Random=true
   )
   &CameraOffset(Camera Offset Amount=50)
  ]
  particleUpdateScript [
   &ParticleState()
   &DynamicMaterialParameters(
    Index 0 Param 1=&FloatFromCurve(FloatCurve=DataInterface() {propertyValues {Curve {keys [{interpMode=RCIM_Cubic, tangentMode=RCTM_User, time=0.5, value=0}, {interpMode=RCIM_Cubic, tangentMode=RCTM_User, time=1, value=1}]}}})
    Index 0 Param 3=1
    Module.Param0WriteEnabled=true
    Module.Param1WriteEnabled=true
    Module.Param2WriteEnabled=true
    Module.Param3WriteEnabled=true
   )
   &ScaleMeshSize(Scale Factor=&VectorFromCurve(VectorCurve=DataInterface() {propertyValues {
       XCurve {keys [
         {arriveTangent=4.1785, interpMode=RCIM_Cubic, leaveTangent=4.1785, tangentMode=RCTM_User, time=0, value=0}
         {arriveTangent=1.1179, interpMode=RCIM_Cubic, leaveTangent=1.1179, tangentMode=RCTM_User, time=0.2042, value=0.6228}
         {arriveTangent=0.0224, interpMode=RCIM_Cubic, leaveTangent=0.0224, tangentMode=RCTM_User, time=1, value=1}
        ]}
       YCurve {keys [
         {arriveTangent=4.1785, interpMode=RCIM_Cubic, leaveTangent=4.1785, tangentMode=RCTM_User, time=0, value=0}
         {arriveTangent=1.1179, interpMode=RCIM_Cubic, leaveTangent=1.1179, tangentMode=RCTM_User, time=0.2042, value=0.6228}
         {arriveTangent=0.0224, interpMode=RCIM_Cubic, leaveTangent=0.0224, tangentMode=RCTM_User, time=1, value=1}
        ]}
       ZCurve {keys [{tangentMode=RCTM_User, time=0, value=1}, {tangentMode=RCTM_User, time=1, value=1}]}
      }}))
   &ScaleColor(
    Scale Alpha=&FloatFromCurve(
     FloatCurve=DataInterface() {propertyValues {Curve {keys [
         {arriveTangent=8.2356, interpMode=RCIM_Cubic, leaveTangent=8.2356, tangentMode=RCTM_User, time=0, value=0.5}
         {arriveTangent=0.1969, interpMode=RCIM_Cubic, leaveTangent=0.1969, tangentMode=RCTM_User, time=0.2223, value=1}
         {arriveTangent=-1.2858, interpMode=RCIM_Cubic, leaveTangent=-1.2858, time=0.4, value=0.85}
         {arriveTangent=-0.7636, interpMode=RCIM_Cubic, leaveTangent=-0.7636, tangentMode=RCTM_User, time=1, value=0}
        ]}}}
     Scale Curve=&FN_FrontEndFloatModifier(Gameplay=0.75, Front End=0.35)
    )
    ScaleA=true
    ScaleRGB=true
   )
   &InheritSourceMovement(ApplyForce=false, ApplyPosition=true, ApplyPositionFalloff=false)
   &SolveForcesAndVelocity(Clamp Velocity=false, Limit Acceleration=false)
  ]
  rendererClasses [&NiagaraMeshRendererProperties]
  renderers [{rendererClass=&NiagaraMeshRendererProperties, rendererIndex=0}]
 }
 Smoke {
  emitterUpdateScript [
   &EmitterState(Life Cycle Mode=::NewEnumerator1, Loop Behavior=::NewEnumerator1, Loop Duration=0.2, UseLoopDelay=false)
   &SpawnBurst_Instantaneous(Spawn Count=25, Spawn Group=1, Use Spawn Probability=false, UseLoopCountLimit=false)
  ]
  particleSpawnScript [
   /SetVariables(Particles.SpawnGroup=$Emitter.SpawnGroup) {isSetParameters=true}
   &PartitionParticles(Number Of Partitions=4)
   &InitializeParticle(
    Color Channel Mode:ENiagaraChannelCorrelation=::NewEnumerator0
    Color Maximum=<rgba 0.35 0.3477 0.336 0.25>
    Color Minimum=<rgba 0.35 0.3404 0.3251 0.2>
    Color Mode=::NewEnumerator2
    Lifetime=&MergeHelper_FloatSwitch(
     0=&RandomRangeFloat(Minimum=0.5, Maximum=0.8)
     1=&RandomRangeFloat(Minimum=0.75, Maximum=1.2)
     2=&RandomRangeFloat(Minimum=0.5, Maximum=0.5)
     3=?
     Index=$Particles.SpawnGroup
     Num=::NewEnumerator1
    )
    Sprite Rotation Mode=::NewEnumerator2
    Sprite Size Mode=::NewEnumerator1
    Uniform Sprite Size Max=35
    Uniform Sprite Size Min=20
    UsePositionOffset=false
    Write Material Random=true
   ) {module=InitializeParticle001}
   &ShapeLocation(
    Cylinder Height=0
    Cylinder Height Midpoint=0
    Cylinder Radius=10
    MaskSpawnGroup=false
    Offset Coordinate Space=::Local
    Offset Mode=::NewEnumerator0
    Override Seed=false
    Shape Primitive=::NewEnumerator1
    Surface Only=false
    UseOwnerScale=true
   )
   &AddVelocity(
    Module.Use Falloff Distance=false
    Override Seed=false
    Velocity Mode=::NewEnumerator1
    Velocity Speed=&MergeHelper_FloatSwitch(
     0=&RandomRangeFloat(Minimum=150, Maximum=500)
     1=&RandomRangeFloat(Minimum=350, Maximum=700)
     2=&RandomRangeFloat(Minimum=300, Maximum=500)
     3=?
     Index=$Particles.SpawnGroup
     Num=::NewEnumerator1
    )
   )
   &ApplyInitialForces()
  ]
  particleUpdateScript [
   /SetVariables(Particles.Jitter:NiagaraFloat=&Waveform("Waveform [1]"=::NewEnumerator12, "[1] Frequency"=25, "[1] Amplitude Min/Max"=<0.5 1>)) {isSetParameters=true}
   &ParticleState()
   &Drag(Drag=7, Use Linear Drag=true, Use Rotational Drag=true)
   &InheritSourceMovement(ApplyForce=false, ApplyPosition=true, ApplyPositionFalloff=false)
   &SolveForcesAndVelocity(Clamp Velocity=false, Limit Acceleration=false)
   &ScaleSpriteSize(Uniform Curve Sprite Scale=DataInterface() {propertyValues {Curve {keys [
        {arriveTangent=1.491, interpMode=RCIM_Cubic, leaveTangent=1.491, tangentMode=RCTM_User, time=0, value=0.5}
        {arriveTangent=0.5, interpMode=RCIM_Cubic, leaveTangent=0.5, time=0.95, value=3}
        <at 1 1 RCIM_Cubic>
       ]}}})
   &DynamicMaterialParameters(
    Index 0 Param 1=&Add_Float(
     A=&RandomRangeFloat(Minimum=0, Maximum=1)
     B=&FloatFromCurve(
      FloatCurve=DataInterface() {propertyValues {Curve {keys [
          <at 0 0 RCIM_Cubic>
          {arriveTangent=1.5947, interpMode=RCIM_Cubic, leaveTangent=1.5947, tangentMode=RCTM_User, time=0.2557, value=0.5257}
          <at 1 1 RCIM_Cubic>
         ]}}}
      CurveIndex=$Particles.Age
      Scale Curve=0.25
     )
    )
    Index 0 Param 3=1
    Index 0 Param 4=1
    Index 2 Param 3=2
    Index 3 Param 1=1
    Index 3 Param 2=70
    Module.Param0WriteEnabled=true
    Module.Param0WriteEnabled1=true
    Module.Param0WriteEnabled2=true
    Module.Param0WriteEnabled3=true
    Module.Param1WriteEnabled=true
    Module.Param1WriteEnabled1=true
    Module.Param1WriteEnabled2=true
    Module.Param1WriteEnabled3=true
    Module.Param2WriteEnabled=true
    Module.Param2WriteEnabled1=true
    Module.Param2WriteEnabled2=true
    Module.Param2WriteEnabled3=true
    Module.Param3WriteEnabled=true
    Module.Param3WriteEnabled1=true
    Module.Param3WriteEnabled2=true
    Module.Param3WriteEnabled3=true
    Write Parameter Index 1=true
    Write Parameter Index 2=true
    Write Parameter Index 3=true
   ) {module=DynamicMaterialParameters001}
   &CameraOffset(Camera Offset Amount=5)
   &ScaleColor(
    Scale Alpha=&Multiply_Float(
     A=&FloatFromCurve(
      FloatCurve=DataInterface() {propertyValues {Curve {keys [{interpMode=RCIM_Cubic, tangentMode=RCTM_User, time=0, value=0}, <at 0.25 1 RCIM_Cubic>, <at 0.9 0>]}}}
      Scale Curve=&FN_FrontEndFloatModifier(Gameplay=0.2, Front End=0.15)
     )
     B=&MergeHelper_FloatSwitch(0=0.5, 1=1, 2=0.5, 3=?, Index=$Particles.SpawnGroup, Num=::NewEnumerator1)
    )
    Scale RGB=<5 5 5>
    ScaleA=true
    ScaleRGB=true
   )
  ]
  rendererClasses [&NiagaraSpriteRendererProperties]
  renderers [
   {rendererClass=&NiagaraSpriteRendererProperties}
   {rendererClass=&NiagaraSpriteRendererProperties, rendererIndex=1}
  ]
 }
}
summary {
 systemName=NS_Emote_DrapeUltra_Leap
 emitters [
  {emitterName=Shockwave, rendererClasses [&NiagaraMeshRendererProperties]}
  {emitterName=Smoke, rendererClasses [&NiagaraSpriteRendererProperties]}
  {emitterName=Air_Trail_Up, rendererClasses [&NiagaraRibbonRendererProperties]}
  {emitterName=AirLines, rendererClasses [&NiagaraSpriteRendererProperties]}
 ]
 userVariables [
  User.bIsFrontEnd {type=NiagaraBool}
  User.bIsFrontEndPreview {type=NiagaraBool}
 ]
}
