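# Niagara system for the orange portal in the DrapeUltra emote.
# Portal spawns two partitioned meshes that grow and shrink over lifetime, with pulsing tint and an animated dynamic material parameter.
# Sparks fires two delayed bursts of orange ring-spawned sprites driven by curl noise and drag, while Air_Trail_Up emits a brief, delayed, fast ribbon streak.
# Front-end modifiers adjust trail alpha and spark speed and size; SpawnStar and DespawnStar show no renderers.
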NiagaraSystem /BRCosmetics/Animation/Game/MainPlayer/Emotes/DrapeUltra/FX/NS_Emote_DrapeUltra_Portal_Orange
dependencies {
 usedDataInterfaces [&NiagaraDataInterfaceCurve, &NiagaraDataInterfaceColorCurve]
 usedDynamicInputs []
 usedModules [&SystemState]
}
emitters {
 Air_Trail_Up {
  emitterUpdateScript [
   &EmitterState(
    Inactive Response=::NewEnumerator1
    Life Cycle Mode=::NewEnumerator1
    Loop Behavior=::NewEnumerator1
    Loop Delay=0.95
    Loop Duration=0.13
    UseLoopDelay=true
   )
   &SpawnRate(SpawnRate=400, Use Spawn Probability=false)
  ]
  particleSpawnScript [
   &InitializeParticle(
    Color=<rgba 5 5 5 1>
    Lifetime=0.1
    Position Offset=<0 5 0>
    Position Offset Coordinate Space=::Local
    Ribbon Width=16
    Ribbon Width Mode=::NewEnumerator1
    UsePositionOffset=true
    Write Material Random=true
   )
   &AddVelocity(Velocity=<-40 1000 -250>, Velocity Speed Scale=0.8)
  ]
  particleUpdateScript [
   &ParticleState()
   &SolveForcesAndVelocity(Clamp Velocity=false, Limit Acceleration=false)
   &ScaleColor(
    Scale Alpha=&FloatFromCurve(
     FloatCurve=DataInterface() {propertyValues {Curve {keys [
         {arriveTangent=-2.5607, interpMode=RCIM_Cubic, leaveTangent=-2.5607, leaveTangentWeight=1.4087, tangentMode=RCTM_User, time=0, value=1}
         {interpMode=RCIM_Cubic, leaveTangent=0.45, time=1, value=0}
        ]}}}
     CurveIndex=$Emitter.NormalizedLoopAge
     Scale Curve=&FN_FrontEndFloatModifier(Gameplay=4, Front End=3)
    )
    Scale RGB=&MakeVectorFromLinearColorRGB(Color=&ColorFromCurve(ColorCurve=DataInterface() {propertyValues {
        AlphaCurve {keys []}
        BlueCurve {keys [
          <at 0 0.3802>
          {arriveTangent=-0.9804, interpMode=RCIM_Cubic, leaveTangent=-0.9804, time=0.0299, value=0.8646}
          <at 0.17 0.2135>
          <at 0.35 0.3125 RCIM_Cubic>
         ]}
        GreenCurve {keys [
          <at 0 0.3802>
          {arriveTangent=-0.9804, interpMode=RCIM_Cubic, leaveTangent=-0.9804, time=0.0299, value=0.8646}
          <at 0.17 0.2135>
          <at 0.35 0.3125 RCIM_Cubic>
         ]}
        RedCurve {keys [
          <at 0 0.3802>
          {arriveTangent=-0.9804, interpMode=RCIM_Cubic, leaveTangent=-0.9804, time=0.0299, value=0.8646}
          <at 0.17 0.2135>
          <at 0.35 0.3125 RCIM_Cubic>
         ]}
       }}))
    ScaleA=true
    ScaleRGB=true
   )
  ]
  rendererClasses [&NiagaraRibbonRendererProperties]
  renderers [{rendererClass=&NiagaraRibbonRendererProperties}]
 }
 DespawnStar {rendererClasses [], renderers []}
 Portal {
  emitterUpdateScript [
   &EmitterState()
   &SpawnBurst_Instantaneous(Spawn Count=2, Use Spawn Probability=false, UseLoopCountLimit=false)
  ]
  particleSpawnScript [
   &PartitionParticles()
   &InitializeParticle(
    Lifetime=1.3
    Mesh Scale Mode=::NewEnumerator3
    Mesh Uniform Scale=&MergeHelper_FloatSwitch(0=0.9, 1=0.47, 2=?, 3=?, Index=$Particles.PartitionParticles.Partition, Num=::NewEnumerator0)
    Position Offset=&MergeHelper_VectorSwitch(0=<0 0 0>, 1=<0 0 0>, 2=?, 3=?, Index=$Particles.PartitionParticles.Partition, Num=::NewEnumerator0)
    Sprite Size Mode=::NewEnumerator3
    Uniform Sprite Size=25
    UsePositionOffset=true
    Write Material Random=true
   )
   /SetVariables(Particles.VisibilityTag=$Particles.PartitionParticles.Partition) {isSetParameters=true}
   &InitialMeshOrientation(
    Mesh Axis To Orient=<0 0 0>
    Mesh Orientation Mode=::NewEnumerator2
    Rotation=&MergeHelper_VectorSwitch(0=<0.43 0 0>, 1=<0.61 0 0>, 2=?, 3=?, Index=$Particles.PartitionParticles.Partition, Num=::NewEnumerator0)
    Use Rotation Vector=true
   ) {enabled=false}
  ]
  particleUpdateScript [
   &ParticleState()
   &ScaleMeshSize(Scale Factor=&VectorFromFloat(Value=&FloatFromCurve(
      FloatCurve=DataInterface() {propertyValues {Curve {
         defaultValue=1
         keys [
          {interpMode=RCIM_Cubic, tangentMode=RCTM_User, time=0, value=0}
          {arriveTangent=-1.7828, interpMode=RCIM_Cubic, leaveTangent=-1.7828, tangentMode=RCTM_User, time=0.3, value=1}
          {arriveTangent=0.0159, interpMode=RCIM_Cubic, leaveTangent=0.0159, tangentMode=RCTM_User, time=0.9018, value=0.988}
          {tangentMode=RCTM_User, time=1, value=0}
         ]
        }}}
      Scale Curve=1.4
     )))
   &ScaleColor(
    Scale RGB=&MergeHelper_VectorSwitch(
     0=&Lerp_Vector(
      A=&MakeVectorFromLinearColorRGB(Color=<rgba 0.9 0.72 0.72 1>)
      B=&MakeVectorFromLinearColorRGB(Color=<rgba 1.07 1.07 0.9095 1>)
      Alpha=&Waveform("Waveform [1]"=::NewEnumerator12, "[1] Frequency"=1.3, "[1] Amplitude Min/Max"=<0 1>)
     )
     1=&Lerp_Vector(
      A=&MakeVectorFromLinearColorRGB(Color=<rgba 0.85 0.68 0.68 1>)
      B=&MakeVectorFromLinearColorRGB(Color=<rgba 1.05 1.05 0.8925 1>)
      Alpha=&Waveform("Waveform [1]"=::NewEnumerator12, "[1] Frequency"=1.4, "[1] Amplitude Min/Max"=<0 1>)
     )
     2=?
     3=?
     Index=$Particles.PartitionParticles.Partition
     Num=::NewEnumerator0
    )
    ScaleA=true
    ScaleRGB=true
   )
   &DynamicMaterialParameters(
    Index 0 Param 1=&FloatFromCurve(
     FloatCurve=DataInterface() {propertyValues {Curve {
        defaultValue=1
        keys [
         {interpMode=RCIM_Cubic, tangentMode=RCTM_User, time=0, value=0}
         {interpMode=RCIM_Cubic, tangentMode=RCTM_User, time=0.3, value=1}
         {arriveTangent=0.0159, interpMode=RCIM_Cubic, leaveTangent=0.0159, tangentMode=RCTM_User, time=0.85, value=1}
         {tangentMode=RCTM_User, time=1, value=0}
        ]
       }}}
     Scale Curve=1.1
    )
    Module.Index0Float0Write=true
    Module.Index0Float1Write=true
    Module.Index0Float2Write=true
    Module.Index0Float3Write=true
   )
   /SetVariables() {isSetParameters=true}
  ]
  rendererClasses [&NiagaraMeshRendererProperties]
  renderers [
   {rendererClass=&NiagaraMeshRendererProperties, rendererIndex=0}
   {rendererClass=&NiagaraMeshRendererProperties, rendererIndex=1}
   {rendererClass=&NiagaraMeshRendererProperties, rendererIndex=2}
   {rendererClass=&NiagaraMeshRendererProperties, rendererIndex=3}
   {rendererClass=&NiagaraMeshRendererProperties, rendererIndex=4}
  ]
 }
 Sparks {
  emitterUpdateScript [
   &EmitterState(Life Cycle Mode=::NewEnumerator1, Loop Behavior=::NewEnumerator1, Loop Delay=0.05, Loop Duration=1, UseLoopDelay=true)
   &SpawnBurst_Instantaneous(Spawn Count=8, Spawn Time=0.1, Use Spawn Probability=false)
   &SpawnBurst_Instantaneous(Spawn Count=4, Spawn Group=1, Spawn Time=0.8, Use Spawn Probability=false) {module=SpawnBurst_Instantaneous001}
  ]
  particleSpawnScript [
   &InitializeParticle(
    Color=<rgba 1 0.475 0.1 1>
    Lifetime=&Multiply_Float(
     A=&MergeHelper_FloatSwitch(0=1, 1=0.6, 2=?, 3=?, Index=&SpawnGroup(), Num=::NewEnumerator0)
     B=&RandomRangeFloat(Minimum=0.3, Maximum=1)
    )
    Sprite Size Max=<1.5 6>
    Sprite Size Min=<1.5 3>
    Sprite Size Mode=::NewEnumerator2
    UsePositionOffset=false
    Write Material Random=true
   )
   &ShapeLocation(
    MaskSpawnGroup=false
    Offset=<0 5 -2>
    Offset Mode=::NewEnumerator0
    Override Seed=false
    Ring Radius=&MergeHelper_FloatSwitch(0=9, 1=7, 2=?, 3=?, Index=&SpawnGroup(), Num=::NewEnumerator0)
    Rotation Axis=<0 0.5 -0.2>
    Rotation Coordinate Space=::Simulation
    Rotation Mode=::NewEnumerator1
    Shape Primitive=::NewEnumerator4
    UseOwnerScale=true
   )
   &AddVelocityFromPoint(
    Module.Use Falloff Distance=false
    Origin Offset Coordinate Space=::Simulation
    Velocity Strength=&Multiply_Float(
     A=&Multiply_Float(
      A=&FN_FrontEndFloatModifier(Gameplay=0.7, Front End=0.35)
      B=&MergeHelper_FloatSwitch(0=1, 1=0.25, 2=?, 3=?, Index=&SpawnGroup(), Num=::NewEnumerator0)
     )
     B=&RandomRangeFloat(Minimum=75, Maximum=200)
    )
   )
   &CameraOffset(Camera Offset Amount=&Multiply_Float(A=30, B=$System.CameraOffsetScale))
   &AddVelocity(
    Rotation Coordinate Space=::Simulation
    Velocity=$Particles.ShapeLocation.ShapeBitangent
    Velocity Speed Scale=&RandomRangeFloat(Minimum=40, Maximum=80)
   )
   &AddVelocity(
    Rotation Coordinate Space=::Simulation
    Velocity=<0 1 -0.5>
    Velocity Speed Scale=&Multiply_Float(
     A=&RandomRangeFloat(Minimum=50, Maximum=110)
     B=&MergeHelper_FloatSwitch(0=1, 1=2, 2=?, 3=?, Index=&SpawnGroup(), Num=::NewEnumerator0)
    )
   ) {module=AddVelocity001}
  ]
  particleUpdateScript [
   &ParticleState()
   &ScaleColor(
    Scale Alpha=0.25
    Scale RGB=&VectorFromFloat(Value=&FloatFromCurve(
      FloatCurve=DataInterface() {propertyValues {Curve {keys [{interpMode=RCIM_Cubic, tangentMode=RCTM_User, time=0, value=10}, <at 1 1>]}}}
      Scale Curve=0.2
     ))
    ScaleA=true
    ScaleRGB=true
   ) {module=ScaleColor001}
   &Drag(
    Drag=&Multiply_Float(
     A=&FloatFromCurve(
      FloatCurve=DataInterface() {propertyValues {Curve {keys [<at 0 0.5>, {arriveTangent=0.5, interpMode=RCIM_Cubic, leaveTangent=0.5, time=0.3, value=0.5}, <at 1 1 RCIM_Cubic>]}}}
      Scale Curve=15
     )
     B=1
    )
    Use Linear Drag=true
    Use Rotational Drag=false
   )
   /Niagara/Modules/Update/Forces/CurlNoiseForce(
    Noise Frequency=1
    Noise Strength=&FloatFromCurve(FloatCurve=DataInterface() {propertyValues {Curve {keys [<at 0 1>,<at 1 0>]}}}, Scale Curve=800)
    Pan Noise=true
    Pan Noise Field=<0.1 0.1 0.5>
   )
   &GravityForce(Gravity=&Multiply_VectorByFloat(
     Vector=&RandomRangeVector(Minimum=<0 0 800>, Maximum=<0 0 -500>)
     Float=&MergeHelper_FloatSwitch(0=1, 1=0.2, 2=?, 3=?, Index=&SpawnGroup(), Num=::NewEnumerator0)
    )) {module=GravityForce001}
   &SolveForcesAndVelocity(Clamp Velocity=false, Limit Acceleration=false)
   &ScaleSpriteSizeBySpeed(Max Scale Factor=<1 6>, Min Scale Factor=<0.5 0.1>, Velocity Threshold=800)
   &ScaleSpriteSize(Scale Factor=&MakeVector2D(
     X=&FloatFromCurve(FloatCurve=DataInterface() {propertyValues {Curve {keys [<at 0 0.5>,<at 0.1 1>,<at 1 0.5>]}}})
     Y=&FloatFromCurve(FloatCurve=DataInterface() {propertyValues {Curve {keys [<at 0 1>,<at 1 0>]}}})
    ))
   &ScaleSpriteSize(
    Scale Sprite Size Mode=::NewEnumerator0
    Uniform Scale Factor=&FN_FrontEndFloatModifier(Gameplay=1, Front End=$User.FrontEndScale)
   ) {module=ScaleSpriteSize001}
   &FN_Sprite_Attributes(
    Color=&MakeLinearColorFromVectorAndFloat(
     "Vector (RGB)"=&MakeVectorFromLinearColorRGB(Color=&ColorFromCurve(ColorCurve=DataInterface() {propertyValues {
         AlphaCurve {keys [<at 0 1 RCIM_Cubic>,<at 1 0 RCIM_Cubic>]}
         BlueCurve {keys [<at 0 0.14 RCIM_Cubic>,<at 0.9637 0.035 RCIM_Cubic>]}
         GreenCurve {keys [<at 0 0.698 RCIM_Cubic>,<at 0.9637 0.1125 RCIM_Cubic>]}
         RedCurve {keys [<at 0 2 RCIM_Cubic>,<at 0.9637 0.5 RCIM_Cubic>]}
        }}))
     "Float (Alpha)"=&FloatFromCurve(FloatCurve=DataInterface() {propertyValues {Curve {keys [<at 0 1>,<at 0.8 0>]}}}, Scale Curve=0.25)
    )
    Create Color Attribute=true
    Create SpriteSize Attribute=true
    SpriteSize=&Multiply_Vector2D_ByFloat(
     Float=&FN_FrontEndFloatModifier(Gameplay=0.85, Front End=0.85)
     Vector2D=&Multiply_Vector2D(
      A=<5 15>
      B=&MakeVector2D(
       X=&FloatFromCurve(FloatCurve=DataInterface() {propertyValues {Curve {keys [<at 0 0.5>,<at 0.1 1>,<at 1 0.5>]}}})
       Y=&FloatFromCurve(FloatCurve=DataInterface() {propertyValues {Curve {keys [<at 0 1>,<at 1 0>]}}})
      )
     )
    )
    Write to Module Namespace=true
   )
  ]
  rendererClasses [&NiagaraSpriteRendererProperties]
  renderers [
   {rendererClass=&NiagaraSpriteRendererProperties}
   {rendererClass=&NiagaraSpriteRendererProperties, rendererIndex=1}
  ]
 }
 SpawnStar {rendererClasses [], renderers []}
}
summary {
 systemName=NS_Emote_DrapeUltra_Portal_Orange
 emitters [
  {emitterName=Portal, rendererClasses [&NiagaraMeshRendererProperties]}
  {emitterName=Sparks, rendererClasses [&NiagaraSpriteRendererProperties]}
  {emitterName=Air_Trail_Up, rendererClasses [&NiagaraRibbonRendererProperties]}
  {emitterName=SpawnStar, rendererClasses []}
  {emitterName=DespawnStar, rendererClasses []}
 ]
 userVariables [
  User.bIsFrontEnd {type=NiagaraBool}
  User.bIsFrontEndPreview {type=NiagaraBool}
  User.FrontEndScale {type=NiagaraFloat, defaultValue=0.5, description=""}
 ]
}
