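# Niagara system for the DrapeUltra emote's fiery sword slash, in orange or blue color schemes selected by User.Variant.
# SwordSlash rotates short-lived meshes with lights, while SplineFire and SplineGlow burst meshes animated through dynamic material parameters.
# Embers and FlameLicks spawn sprites along a ring arc swept by emitter age, flung along its tangent.
# User.ScaleLifetime stretches particle lifetimes, and front-end and time-of-day modifiers adjust size and brightness.
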NiagaraSystem /BRCosmetics/Animation/Game/MainPlayer/Emotes/DrapeUltra/FX/NS_Emote_DrapeUltra_Slashes
dependencies {
 usedDataInterfaces [
  &NiagaraDataInterfaceCurve
  &NiagaraDataInterfaceColorCurve
  &NiagaraDataInterfaceSpriteRendererInfo
  &NiagaraDataInterfaceSkeletalMesh
  &NiagaraDataInterfaceSocketReader
 ]
 usedDynamicInputs [&SetBoolByFloatComparison]
 usedModules [
  &SampleSkeletalMesh
  &SystemState
  &self:SystemSpawnScript.SystemScriptSource.SystemScriptGraph.NiagaraNodeAssignment_0.SetVariables_#1
  &self:SystemSpawnScript.SystemScriptSource.SystemScriptGraph.NiagaraNodeAssignment_1.SetVariables_#2
  &self:SystemSpawnScript.SystemScriptSource.SystemScriptGraph.NiagaraNodeAssignment_3.SetVariables_#3
 ]
}
emitters {
 Embers {
  emitterUpdateScript [
   &EmitterState(Life Cycle Mode=::NewEnumerator1, Loop Behavior=::NewEnumerator1, Loop Delay=0.15, Loop Duration=0.1, UseLoopDelay=false)
   &SpawnRate(SpawnRate=400, Use Spawn Probability=false)
  ]
  particleSpawnScript [
   &InitializeParticle(
    Lifetime=&Multiply_Float(A=$User.ScaleLifetime, B=&RandomRangeFloat(Minimum=1, Maximum=1.6))
    Mesh Scale Mode=::NewEnumerator1
    Mesh Uniform Scale Max=1.5
    Mesh Uniform Scale Min=0.5
    Sprite Size Max=<3 10>
    Sprite Size Min=<3 5>
    Sprite Size Mode=::NewEnumerator2
    UsePositionOffset=true
    Write Material Random=true
   )
   &ShapeLocation(
    MaskSpawnGroup=false
    Offset=<30 0 55>
    Offset Mode=::NewEnumerator0
    Override Seed=false
    Radius Position:NiagaraFloat=&RandomRangeFloat(Minimum=0.8, Maximum=1)
    Ring / Disc Distribution Mode:ENiagaraSphereDistributionMode=::NewEnumerator1
    Ring Radius=150
    Rotation Axis=<0 1 0>
    Rotation Coordinate Space=::Simulation
    Rotation Mode=::NewEnumerator1
    Shape Primitive=::NewEnumerator4
    U Position=&FloatFromCurve(FloatCurve=DataInterface() {propertyValues {Curve {keys [<at 0.02 -0.25>,<at 0.12 0.28>]}}}, CurveIndex=$Emitter.Age)
    UseOwnerScale=true
   )
   &AddVelocity(
    Cone Angle=25
    Cone Axis=$Particles.ShapeLocation.ShapeBitangent
    Module.UseConeFalloff=false
    Module.UseVelDistribution=false
    Override Seed=false
    Velocity Mode=::NewEnumerator2
    Velocity Speed=&RandomRangeFloat(Minimum=-150, Maximum=-400)
   ) {module=AddVelocity001}
  ]
  particleUpdateScript [
   &ParticleState()
   &ScaleSpriteSize(
    Uniform Curve Scale=&Multiply_Float(A=0.8, B=$User.FrontEndScale)
    Uniform Curve Sprite Scale=DataInterface() {propertyValues {Curve {keys [<at 0 0 RCIM_Cubic>,<at 0.2 1 RCIM_Cubic>,<at 1 0 RCIM_Cubic>]}}}
   )
   &ScaleColor(
    Scale Alpha=&Multiply_Float(
     A=1
     B=&FloatFromCurve(FloatCurve=DataInterface() {propertyValues {Curve {keys [
          {interpMode=RCIM_Cubic, tangentMode=RCTM_User, time=0.1, value=0}
          {arriveTangent=-1.3622, interpMode=RCIM_Cubic, leaveTangent=-1.3622, tangentMode=RCTM_User, time=0.2, value=1}
          {tangentMode=RCTM_User, time=0.8, value=0}
         ]}}})
    )
    Scale RGB=&MakeVectorFromLinearColorRGB(Color=&MergeHelper_ColorSwitch(
      0=&ColorFromCurve(ColorCurve=DataInterface() {propertyValues {
         AlphaCurve {keys [<at 0.0926 1 RCIM_Cubic>,<at 1 0 RCIM_Cubic>]}
         BlueCurve {keys [<at 0 0.135 RCIM_Cubic>,<at 0.9709 0.0267>]}
         GreenCurve {keys [<at 0 2.5675 RCIM_Cubic>,<at 0.9709 0.1078>]}
         RedCurve {keys [<at 0 5 RCIM_Cubic>,<at 0.9709 1>]}
        }})
      1=&ColorFromCurve(ColorCurve=DataInterface() {propertyValues {
         AlphaCurve {keys [<at 0.0926 1 RCIM_Cubic>,<at 1 0 RCIM_Cubic>]}
         BlueCurve {keys [<at 0 5 RCIM_Cubic>,<at 0.9709 1>]}
         GreenCurve {keys [<at 0 2.5675 RCIM_Cubic>,<at 0.9709 0.1078>]}
         RedCurve {keys [<at 0 0.135 RCIM_Cubic>,<at 0.9709 0.0267>]}
        }})
      2=?
      3=?
      Index=$User.Variant
      Num=::NewEnumerator0
     ))
    ScaleA=true
    ScaleRGB=true
   )
   &Drag(
    Drag=&Multiply_Float(
     A=&FloatFromCurve(
      FloatCurve=DataInterface() {propertyValues {Curve {keys [<at 0 0.5>, {arriveTangent=0.5, interpMode=RCIM_Cubic, leaveTangent=0.5, time=0.3, value=0.5}, <at 1 1 RCIM_Cubic>]}}}
      Scale Curve=10
     )
     B=1
    )
    Use Linear Drag=true
    Use Rotational Drag=false
   ) {module=Drag001}
   /Niagara/Modules/Update/Forces/CurlNoiseForce(
    Noise Frequency=1
    Noise Strength=&FloatFromCurve(FloatCurve=DataInterface() {propertyValues {Curve {keys [<at 0 1>,<at 1 0>]}}}, Scale Curve=200)
    Pan Noise=true
    Pan Noise Field=<0.2 0.2 0.5>
   ) {module=CurlNoiseForce001}
   &ScaleSpriteSize(Scale Factor=&MakeVector2D(
     X=&FloatFromCurve(FloatCurve=DataInterface() {propertyValues {Curve {keys [<at 0 0.5>,<at 0.1 1>,<at 1 0.5>]}}})
     Y=&FloatFromCurve(FloatCurve=DataInterface() {propertyValues {Curve {keys [<at 0 1>,<at 1 0>]}}})
    )) {module=ScaleSpriteSize001}
   &ScaleSpriteSize(
    Scale Sprite Size Mode=::NewEnumerator0
    Uniform Scale Factor=&FN_FrontEndFloatModifier(Gameplay=2, Front End=$User.FrontEndScale)
   ) {module=ScaleSpriteSize002}
   &SolveForcesAndVelocity(Clamp Velocity=false, Limit Acceleration=false)
   &ScaleSpriteSizeBySpeed(Max Scale Factor=<1 8>, Min Scale Factor=<0.5 0.1>, Velocity Threshold=800)
   &FN_Sprite_Attributes(
    Color=&MakeLinearColorFromVectorAndFloat(
     "Vector (RGB)"=&MakeVectorFromLinearColorRGB(Color=&MergeHelper_ColorSwitch(
       0=&ColorFromCurve(ColorCurve=DataInterface() {propertyValues {
          AlphaCurve {keys [<at 0 1 RCIM_Cubic>,<at 1 0 RCIM_Cubic>]}
          BlueCurve {keys [<at 0 0.14 RCIM_Cubic>,<at 1 0.035 RCIM_Cubic>]}
          GreenCurve {keys [<at 0 0.605 RCIM_Cubic>,<at 1 0.035 RCIM_Cubic>]}
          RedCurve {keys [<at 0 2 RCIM_Cubic>,<at 1 0.5 RCIM_Cubic>]}
         }})
       1=&ColorFromCurve(ColorCurve=DataInterface() {propertyValues {
          AlphaCurve {keys [<at 0 1 RCIM_Cubic>,<at 1 0 RCIM_Cubic>]}
          BlueCurve {keys [<at 0 2 RCIM_Cubic>,<at 0.9637 0.5 RCIM_Cubic>]}
          GreenCurve {keys [<at 0 0.915 RCIM_Cubic>,<at 0.9637 0.0737 RCIM_Cubic>]}
          RedCurve {keys [<at 0 0.14 RCIM_Cubic>,<at 0.9637 0.035 RCIM_Cubic>]}
         }})
       2=?
       3=?
       Index=$User.Variant
       Num=::NewEnumerator0
      ))
     "Float (Alpha)"=&FloatFromCurve(FloatCurve=DataInterface() {propertyValues {Curve {keys [<at 0 1>,<at 0.8 0>]}}}, Scale Curve=0.2)
    )
    Create Color Attribute=true
    Create SpriteSize Attribute=true
    SpriteSize=&Multiply_Vector2D_ByFloat(
     Float=&FN_FrontEndFloatModifier(Gameplay=2, Front End=2)
     Vector2D=&Multiply_Vector2D(
      A=<5 8>
      B=&MakeVector2D(
       X=&FloatFromCurve(FloatCurve=DataInterface() {propertyValues {Curve {keys [<at 0 0.5>,<at 0.1 1>,<at 1 0.5>]}}})
       Y=&FloatFromCurve(FloatCurve=DataInterface() {propertyValues {Curve {keys [<at 0 1>,<at 1 0>]}}})
      )
     )
    )
    Write to Module Namespace=true
   )
  ]
  rendererClasses [&NiagaraSpriteRendererProperties]
  renderers [
   {rendererClass=&NiagaraSpriteRendererProperties}
   {rendererClass=&NiagaraSpriteRendererProperties, rendererIndex=1}
  ]
 }
 FlameLicks {
  emitterUpdateScript [
   &EmitterState(Life Cycle Mode=::NewEnumerator1, Loop Behavior=::NewEnumerator1, Loop Delay=0.15, Loop Duration=0.1, UseLoopDelay=false)
   &SpawnRate(SpawnRate=200, Use Spawn Probability=false)
  ]
  particleSpawnScript [
   &PartitionParticles() {module=PartitionParticles001}
   /SetVariables(Particles.Random size:NiagaraFloat=&RandomRangeFloat(Minimum=0.2, Maximum=0.7)) {isSetParameters=true}
   &InitializeParticle(
    Color=&MergeHelper_ColorSwitch(0=<rgba 1 0.3667 0.05 1>, 1=<rgba 0.05 0.525 1 1>, 2=?, 3=?, Index=$User.Variant, Num=::NewEnumerator0)
    Lifetime=&Multiply_Float(A=$User.ScaleLifetime, B=&RandomRangeFloat(Minimum=0.8, Maximum=1))
    Sprite Rotation Mode=::NewEnumerator2
    Sprite Size=&Multiply_Vector2D_ByFloat(
     Float=&Multiply_Float(A=15, B=&RandomRangeFloat(Minimum=0.1, Maximum=1))
     Vector2D=<1.25 3>
    )
    Sprite Size Mode=::NewEnumerator4
    UsePositionOffset=false
    Write Material Random=true
   )
   &ShapeLocation(
    MaskSpawnGroup=false
    Offset=<30 0 55>
    Offset Mode=::NewEnumerator0
    Override Seed=false
    Radius Position:NiagaraFloat=&RandomRangeFloat(Minimum=0.8, Maximum=1)
    Ring / Disc Distribution Mode:ENiagaraSphereDistributionMode=::NewEnumerator1
    Ring Radius=150
    Rotation Axis=<0 1 0>
    Rotation Coordinate Space=::Simulation
    Rotation Mode=::NewEnumerator1
    Shape Primitive=::NewEnumerator4
    U Position=&FloatFromCurve(FloatCurve=DataInterface() {propertyValues {Curve {keys [<at 0.02 -0.25>,<at 0.12 0.28>]}}}, CurveIndex=$Emitter.Age)
    UseOwnerScale=true
   )
   &AddVelocity(
    Cone Angle=25
    Cone Axis=$Particles.ShapeLocation.ShapeBitangent
    Module.UseConeFalloff=false
    Module.UseVelDistribution=false
    Override Seed=false
    Velocity Mode=::NewEnumerator2
    Velocity Speed=&RandomRangeFloat(Minimum=-100, Maximum=-250)
   ) {module=AddVelocity001}
   &DynamicMaterialParameters(
    Index 0 Param 2=2
    Index 0 Param 4=1
    Module.Param0WriteEnabled=true
    Module.Param1WriteEnabled=true
    Module.Param2WriteEnabled=false
    Module.Param3WriteEnabled=true
   ) {module=DynamicMaterialParameters001}
  ]
  particleUpdateScript [
   &ParticleState()
   &Drag(Drag=4, Use Linear Drag=true, Use Rotational Drag=false)
   /Niagara/Modules/Update/Forces/V2/CurlNoiseForce(Bias Noise=false, Noise Frequency=5, Noise Quality / Cost=::NewEnumerator0, Pan Noise=true)
   &SpriteRotationRate(Rotation Rate=&RandomRangeFloat(Minimum=-250, Maximum=250)) {module=SpriteRotationRate001}
   &ScaleSpriteSize(
    Scale Sprite Size Mode=::NewEnumerator0
    Uniform Scale Factor=&FloatFromCurve(
     FloatCurve=DataInterface() {propertyValues {Curve {
        defaultValue=1
        keys [
         {arriveTangent=-1.9355, interpMode=RCIM_Cubic, leaveTangent=-1.9355, tangentMode=RCTM_User, time=0, value=1}
         {interpMode=RCIM_Cubic, tangentMode=RCTM_User, time=1, value=0}
        ]
       }}}
     Scale Curve=2.5
    )
   ) {module=ScaleSpriteSize001}
   &ScaleColor(
    Scale Alpha=&FloatFromCurve(FloatCurve=DataInterface() {propertyValues {Curve {
        defaultValue=0
        keys [
         {arriveTangent=-2.5249, interpMode=RCIM_Cubic, leaveTangent=-2.5249, tangentMode=RCTM_User, time=0, value=0.9847}
         {arriveTangent=-0.0405, interpMode=RCIM_Cubic, leaveTangent=-0.0405, tangentMode=RCTM_User, time=1, value=0}
        ]
       }}})
    Scale RGB=&VectorFromFloat(Value=&Lerp_Float(
      A=0
      B=&RandomRangeFloat(Minimum=3, Maximum=10)
      Alpha=&MakeFloatFromVector(
       Vector=&VectorFromFloat(Value=&FloatFromCurve(FloatCurve=DataInterface() {propertyValues {Curve {
            defaultValue=1
            keys [
             {arriveTangent=9.0725, interpMode=RCIM_Cubic, leaveTangent=9.0725, tangentMode=RCTM_User, time=0, value=0}
             {arriveTangent=-2.5249, interpMode=RCIM_Cubic, leaveTangent=-2.5249, tangentMode=RCTM_User, time=0.3924, value=0.933}
             {arriveTangent=-0.0405, interpMode=RCIM_Cubic, leaveTangent=-0.0405, tangentMode=RCTM_User, time=1, value=0}
            ]
           }}}))
       Channel=::NewEnumerator0
      )
     ))
    ScaleA=true
    ScaleRGB=true
   )
   &FN_Sprite_Attributes(
    Color=&Multiply_LinearColorByFloat(Linear Color=$Particles.Color, Float=0.3)
    Create Color Attribute=true
    Create SpriteSize Attribute=true
    SpriteSize=&Vector2DFromFloat(Value=&Multiply_Float(A=10, B=&Multiply_Float(A=1, B=1)))
    Write to Module Namespace=true
   ) {module=FN_Sprite_Attributes001}
   &SubUVAnimation(
    End Frame Range Override=12
    Evaluation Type=::NewEnumerator0
    "Sprite Renderer "=DataInterface() {propertyValues {SpriteRenderer=&self:FlameLicks_0.NiagaraSpriteRendererProperties_0}}
    Start Frame Range Override=8
    SubUV Animation Mode:ENiagara_SubUVLookupModeV2=::NewEnumerator2
    UseEndFrame=false
    UseStartFrame=false
   )
   &PointAttractionForce(
    Attraction Radius=200
    Attractor Position:NiagaraPosition=&AddVectorToPosition(Position=<0 0 0>, Vector=<0 0 0>)
    Kill Within Radius=false
    Module.Use Falloff=true
    UseAttractorOffset=false
   )
   &SolveForcesAndVelocity(Clamp Velocity=false, Limit Acceleration=false)
  ]
  rendererClasses [&NiagaraSpriteRendererProperties]
  renderers [
   {rendererClass=&NiagaraSpriteRendererProperties}
   {rendererClass=&NiagaraSpriteRendererProperties, rendererIndex=1}
  ]
 }
 SplineFire {
  emitterUpdateScript [
   &EmitterState(Life Cycle Mode=::NewEnumerator1, Loop Behavior=::NewEnumerator1, Loop Duration=10, UseLoopDelay=false)
   &SpawnBurst_Instantaneous(
    Spawn Count=&MakeCustomIntFromBool(Bool=&InvertBool(Boolean:NiagaraBool=$System.SoftVer), True INT=1, False INT=0)
    Use Spawn Probability=false
    UseLoopCountLimit=false
   )
  ]
  particleSpawnScript [
   &InitializeParticle(
    Color=&MergeHelper_ColorSwitch(0=<rgba 3 1.04 0.06 3>, 1=<rgba 0.06 1.53 3 3>, 2=?, 3=?, Index=$User.Variant, Num=::NewEnumerator0)
    Lifetime=&Multiply_Float(A=0.65, B=$User.ScaleLifetime)
    UsePositionOffset=false
    Write Material Random=true
   )
   /SetVariables(Particles.MeshIndex=&MakeIntFromFloat(Float=0)) {isSetParameters=true}
  ]
  particleUpdateScript [
   &ParticleState()
   &DynamicMaterialParameters(
    Index 0 Param 1=&FloatFromCurve(
     FloatCurve=DataInterface() {propertyValues {Curve {
        defaultValue=1
        keys [
         {tangentMode=RCTM_User, time=0, value=1}
         {arriveTangent=-0.0405, interpMode=RCIM_Cubic, leaveTangent=-0.0405, tangentMode=RCTM_User, time=0.8, value=1}
         {arriveTangent=4.0917, interpMode=RCIM_Cubic, leaveTangent=4.0917, tangentMode=RCTM_User, time=1, value=1.2}
        ]
       }}}
     Scale Curve=0.6
    )
    Index 0 Param 3=&FloatFromCurve(
     FloatCurve=DataInterface() {propertyValues {Curve {
        defaultValue=1
        keys [
         {tangentMode=RCTM_User, time=0, value=0}
         {arriveTangent=-0.0405, interpMode=RCIM_Cubic, leaveTangent=-0.0405, tangentMode=RCTM_User, time=0.6, value=0}
         {arriveTangent=16.8033, interpMode=RCIM_Cubic, leaveTangent=16.8033, tangentMode=RCTM_User, time=1, value=1}
        ]
       }}}
     Scale Curve=7
    )
    Index 0 Param 4=&Add_Float(
     A=&FloatFromCurve(
      FloatCurve=DataInterface() {propertyValues {Curve {
         defaultValue=1
         keys [
          {arriveTangent=9.0725, interpMode=RCIM_Cubic, leaveTangent=9.0725, tangentMode=RCTM_User, time=0, value=0}
          {arriveTangent=-2.5249, interpMode=RCIM_Cubic, leaveTangent=-2.5249, tangentMode=RCTM_User, time=0.06, value=0.9847}
          {arriveTangent=-0.0405, interpMode=RCIM_Cubic, leaveTangent=-0.0405, tangentMode=RCTM_User, time=1, value=0}
         ]
        }}}
      Scale Curve=-0.5
     )
     B=&RandomRangeFloat(Minimum=0, Maximum=2)
    )
    Module.Index0Float0Write=true
    Module.Index0Float1Write=true
    Module.Index0Float2Write=true
    Module.Index0Float3Write=true
   )
   &ScaleColor(
    Scale Alpha=&FloatFromCurve(
     FloatCurve=DataInterface() {propertyValues {Curve {
        defaultValue=0
        keys [
         {arriveTangent=0.2081, interpMode=RCIM_Cubic, leaveTangent=0.2081, tangentMode=RCTM_User, time=0.2, value=1}
         {arriveTangent=-0.2972, interpMode=RCIM_Cubic, leaveTangent=-0.2972, tangentMode=RCTM_User, time=1, value=0}
        ]
       }}}
     Scale Curve=&Multiply_Float(
      A=&FNTimeOfDayScalar(Night Scale=0.4, Morning Scale=0.8, Evening Scale=0.8)
      B=&FN_FrontEndFloatModifier(Gameplay=0.5, Front End=0.3)
     )
    )
    ScaleA=true
    ScaleRGB=true
   )
   /SetVariables(Particles.SoftColor:LinearColor=&MultiplyLinearColor(
     Linear Color=&MergeHelper_ColorSwitch(0=<rgba 0.6 0.2 0 1>, 1=<rgba 0 0.1 0.6 1>, 2=?, 3=?, Index=$User.Variant, Num=::NewEnumerator0)
     Scale Factor=&MakeVector4FromVectorAndFloat(
      "Vector (RGB)"=<1 1 1>
      "Float (Alpha)"=&FloatFromCurve(
       FloatCurve=DataInterface() {propertyValues {Curve {
          defaultValue=0
          keys [
           {arriveTangent=0.2081, interpMode=RCIM_Cubic, leaveTangent=0.2081, tangentMode=RCTM_User, time=0.2, value=1}
           {arriveTangent=-0.2972, interpMode=RCIM_Cubic, leaveTangent=-0.2972, tangentMode=RCTM_User, time=1, value=0}
          ]
         }}}
       Scale Curve=&Multiply_Float(
        A=&FNTimeOfDayScalar(Night Scale=0.6, Morning Scale=0.8, Evening Scale=0.8)
        B=&FN_FrontEndFloatModifier(Gameplay=3, Front End=1.5)
       )
      )
     )
    )) {isSetParameters=true}
   /SetVariables(Particles.DynSoft:Vector4f=&MakeVector4(
     X=&FloatFromCurve(FloatCurve=DataInterface() {propertyValues {Curve {
         defaultValue=1
         keys [
          {tangentMode=RCTM_User, time=0, value=1}
          {arriveTangent=-0.0405, interpMode=RCIM_Cubic, leaveTangent=-0.0405, tangentMode=RCTM_User, time=0.6, value=1}
          {arriveTangent=4.0917, interpMode=RCIM_Cubic, leaveTangent=4.0917, tangentMode=RCTM_User, time=1, value=2}
         ]
        }}})
     Y=0
     Z=&FloatFromCurve(
      FloatCurve=DataInterface() {propertyValues {Curve {
         defaultValue=1
         keys [
          {tangentMode=RCTM_User, time=0, value=0}
          {arriveTangent=-0.0405, interpMode=RCIM_Cubic, leaveTangent=-0.0405, tangentMode=RCTM_User, time=0.6, value=0}
          {arriveTangent=5.6743, interpMode=RCIM_Cubic, leaveTangent=5.6743, tangentMode=RCTM_User, time=1, value=1}
         ]
        }}}
      Scale Curve=20
     )
     W=&Add_Float(
      A=&FloatFromCurve(
       FloatCurve=DataInterface() {propertyValues {Curve {
          defaultValue=1
          keys [
           {arriveTangent=9.0725, interpMode=RCIM_Cubic, leaveTangent=9.0725, tangentMode=RCTM_User, time=0, value=0}
           {arriveTangent=-2.5249, interpMode=RCIM_Cubic, leaveTangent=-2.5249, tangentMode=RCTM_User, time=0.06, value=0.9847}
           {arriveTangent=-0.0405, interpMode=RCIM_Cubic, leaveTangent=-0.0405, tangentMode=RCTM_User, time=1, value=0}
          ]
         }}}
       Scale Curve=-0.8
      )
      B=&RandomRangeFloat(Minimum=0, Maximum=1)
     )
    )) {isSetParameters=true}
   /SetVariables(Particles.DynSoft2:Vector4f=&MakeVector4(
     X=&FloatFromCurve(
      FloatCurve=DataInterface() {propertyValues {Curve {
         defaultValue=1
         keys [
          {tangentMode=RCTM_User, time=0, value=1}
          {arriveTangent=-0.0405, interpMode=RCIM_Cubic, leaveTangent=-0.0405, tangentMode=RCTM_User, time=0.6, value=1}
          {arriveTangent=4.0917, interpMode=RCIM_Cubic, leaveTangent=4.0917, tangentMode=RCTM_User, time=1, value=2}
         ]
        }}}
      Scale Curve=0.9
     )
     Y=0
     Z=&FloatFromCurve(
      FloatCurve=DataInterface() {propertyValues {Curve {
         defaultValue=1
         keys [
          {tangentMode=RCTM_User, time=0, value=0}
          {arriveTangent=-0.0405, interpMode=RCIM_Cubic, leaveTangent=-0.0405, tangentMode=RCTM_User, time=0.6, value=0}
          {arriveTangent=10.8261, interpMode=RCIM_Cubic, leaveTangent=10.8261, tangentMode=RCTM_User, time=1, value=1}
         ]
        }}}
      Scale Curve=12
     )
     W=&Add_Float(
      A=&FloatFromCurve(
       FloatCurve=DataInterface() {propertyValues {Curve {
          defaultValue=1
          keys [
           {arriveTangent=9.0725, interpMode=RCIM_Cubic, leaveTangent=9.0725, tangentMode=RCTM_User, time=0, value=0}
           {arriveTangent=-2.5249, interpMode=RCIM_Cubic, leaveTangent=-2.5249, tangentMode=RCTM_User, time=0.06, value=0.9847}
           {arriveTangent=-0.0405, interpMode=RCIM_Cubic, leaveTangent=-0.0405, tangentMode=RCTM_User, time=1, value=0}
          ]
         }}}
       Scale Curve=-0.65
      )
      B=&RandomRangeFloat(Minimum=0, Maximum=1)
     )
    )) {isSetParameters=true}
  ]
  rendererClasses [&NiagaraMeshRendererProperties]
  renderers [
   {rendererClass=&NiagaraMeshRendererProperties, rendererIndex=0}
   {rendererClass=&NiagaraMeshRendererProperties, rendererIndex=1}
   {rendererClass=&NiagaraMeshRendererProperties, rendererIndex=2}
  ]
 }
 SplineGlow {
  emitterUpdateScript [
   &EmitterState(Life Cycle Mode=::NewEnumerator1, Loop Behavior=::NewEnumerator1, Loop Duration=10, UseLoopDelay=false)
   &SpawnBurst_Instantaneous(Use Spawn Probability=false, UseLoopCountLimit=false)
  ]
  particleSpawnScript [
   &InitializeParticle(
    Color=&MergeHelper_ColorSwitch(0=<rgba 2 0.48 0.1 0.7>, 1=<rgba 0.1 0.7333 2 0.7>, 2=?, 3=?, Index=$User.Variant, Num=::NewEnumerator0)
    Lifetime=&Multiply_Float(A=1.5, B=$User.ScaleLifetime)
    UsePositionOffset=false
    Write Material Random=true
   )
   /SetVariables(Particles.MeshIndex=&MakeIntFromFloat(Float=0)) {isSetParameters=true}
   /SetVariables(Particles.VisibilityTag=$User.Variant) {enabled=false, isSetParameters=true}
  ]
  particleUpdateScript [
   &ParticleState()
   &DynamicMaterialParameters(
    Index 0 Param 1=1
    Index 0 Param 2=&FloatFromCurve(FloatCurve=DataInterface() {propertyValues {Curve {keys [
         {interpMode=RCIM_Cubic, leaveTangent=2.4962, tangentMode=RCTM_Break, time=0.25, value=0}
         {arriveTangent=2.6266, interpMode=RCIM_Cubic, tangentMode=RCTM_Break, time=0.8847, value=1}
        ]}}})
    Index 0 Param 3=&FloatFromCurve(
     FloatCurve=DataInterface() {propertyValues {Curve {keys [{interpMode=RCIM_Cubic, leaveTangent=4.0629, tangentMode=RCTM_Break, time=0.2279, value=0}, <at 1 1>]}}}
     Scale Curve=-5
    )
    Index 0 Param 4=1
    Module.Index0Float0Write=true
    Module.Index0Float1Write=true
    Module.Index0Float2Write=true
    Module.Index0Float3Write=true
   )
   &ScaleColor(
    Scale Alpha=&FloatFromCurve(
     FloatCurve=DataInterface() {propertyValues {Curve {keys [
         {interpMode=RCIM_Cubic, leaveTangent=-3.4444, tangentMode=RCTM_Break, time=0.2255, value=1}
         {interpMode=RCIM_Cubic, tangentMode=RCTM_Break, time=1, value=0}
        ]}}}
     Scale Curve=&FN_FrontEndFloatModifier(Gameplay=0.15, Front End=0.1)
    )
    ScaleA=true
    ScaleRGB=true
   )
  ]
  rendererClasses [&NiagaraMeshRendererProperties]
  renderers [
   {rendererClass=&NiagaraMeshRendererProperties, rendererIndex=0}
   {rendererClass=&NiagaraMeshRendererProperties, rendererIndex=1}
   {rendererClass=&NiagaraMeshRendererProperties, rendererIndex=2}
   {rendererClass=&NiagaraMeshRendererProperties, rendererIndex=3}
  ]
 }
 SwordSlash {
  emitterUpdateScript [
   &EmitterState(Life Cycle Mode=::NewEnumerator1, Loop Behavior=::NewEnumerator1, Loop Duration=10, UseLoopDelay=false)
   &SpawnBurst_Instantaneous(Use Spawn Probability=true, UseLoopCountLimit=false)
  ]
  particleSpawnScript [
   &InitializeParticle(
    Lifetime=&MakeCustomFloatFromBool(Bool=false, True Float=5, False Float=&Multiply_Float(A=0.14, B=$User.ScaleLifetime))
    Mesh Scale=<1 1.3 1>
    Mesh Scale Mode=::NewEnumerator4
    Position Offset=<20 0 50>
    UsePositionOffset=true
    Write Material Random=true
   )
   /SetVariables(Particles.VisibilityTag=$User.Variant) {isSetParameters=true}
  ]
  particleUpdateScript [
   &ParticleState()
   &UpdateMeshOrientation(Rotation Rate=$System.RotationSpeed, Rotation Vector=<0 1 0>)
   &DynamicMaterialParameters(Module.Index0Float0Write=true, Module.Index0Float1Write=true, Module.Index0Float2Write=true, Module.Index0Float3Write=true)
   &ScaleColor(
    Scale Alpha=&FloatFromCurve(
     FloatCurve=DataInterface() {propertyValues {Curve {keys [
         {interpMode=RCIM_Cubic, leaveTangent=-2.1183, tangentMode=RCTM_Break, time=0.05, value=1}
         {interpMode=RCIM_Cubic, tangentMode=RCTM_Break, time=1, value=0}
        ]}}}
     Scale Curve=&FN_FrontEndFloatModifier(
      Gameplay=&FNTimeOfDayScalar(Night Scale=0.5, Morning Scale=0.75, Day Scale=0.9, Evening Scale=0.75)
      Front End=0.8
     )
    )
    ScaleA=true
    ScaleRGB=true
   )
   /SetVariables(Particles.Position2:NiagaraPosition=&AddVectorToPosition(Position=$Particles.Position, Vector=$System.PositionOffset)) {isSetParameters=true}
   &Light_Attributes(
    Color=1
    Color Value To Scale:LinearColor=&MultiplyLinearColor(
     Linear Color=&MergeHelper_ColorSwitch(0=<rgba 1 0.4167 0 1>, 1=<rgba 0 0.5 1 1>, 2=?, 3=?, Index=$User.Variant, Num=::NewEnumerator0)
     Scale Factor=&MakeVector4_FromLinearColor(LinearColor=&ColorFromCurve(ColorCurve=DataInterface() {propertyValues {
         AlphaCurve {keys [<at 0 1>]}
         BlueCurve {keys [<at 0 0>,<at 0.0338 0.1777>,<at 0.123 1>,<at 0.3195 0.1208>,<at 0.9964 0>]}
         GreenCurve {keys [<at 0 0>,<at 0.0338 0.1777>,<at 0.123 1>,<at 0.3195 0.1208>,<at 0.9964 0>]}
         RedCurve {keys [<at 0 0>,<at 0.0338 0.1777>,<at 0.123 1>,<at 0.3195 0.1208>,<at 0.9964 0>]}
        }}))
    )
    Diffuse Scale=1
    Falloff Exponent=1
    Light Exponent=5
    Light Position Offset:Vector3f=&MergeHelper_VectorSwitch(0=<30 0 -100>, 1=<30 0 -100>, 2=?, 3=?, Index=$Particles.VisibilityTag, Num=::NewEnumerator0)
    Light Radius=300
    Position=1
    Position Offset Coordinate Space:ENiagaraCoordinateSpace=::Local
    Radius=1
    Scale Alpha:NiagaraFloat=&FN_Select_Float_from_Time_Of_Day(Morning=0.6, Day=1, Evening=0.6, Night=0.4)
    Scale RGB:Vector3f=&VectorFromFloat(Value=50)
    ScaleA=true
    ScaleRGB=true
    Specular Scale=1
    UsePositionOffset=true
   )
  ]
  rendererClasses [&NiagaraMeshRendererProperties, &NiagaraLightRendererProperties]
  renderers [
   {rendererClass=&NiagaraMeshRendererProperties, rendererIndex=0}
   {rendererClass=&NiagaraMeshRendererProperties, rendererIndex=1}
   {rendererClass=&NiagaraMeshRendererProperties, rendererIndex=2}
   {rendererClass=&NiagaraLightRendererProperties, rendererIndex=3}
   {rendererClass=&NiagaraMeshRendererProperties, rendererIndex=4}
   {rendererClass=&NiagaraMeshRendererProperties, rendererIndex=5}
   {rendererClass=&NiagaraMeshRendererProperties, rendererIndex=6}
   {rendererClass=&NiagaraLightRendererProperties, rendererIndex=7}
  ]
 }
}
summary {
 systemName=NS_Emote_DrapeUltra_Slashes
 emitters [
  {emitterName=SplineGlow, rendererClasses [&NiagaraMeshRendererProperties]}
  {emitterName=SwordSlash, rendererClasses [&NiagaraMeshRendererProperties, &NiagaraLightRendererProperties]}
  {emitterName=FlameLicks, rendererClasses [&NiagaraSpriteRendererProperties]}
  {emitterName=SplineFire, rendererClasses [&NiagaraMeshRendererProperties]}
  {emitterName=Embers, rendererClasses [&NiagaraSpriteRendererProperties]}
 ]
 userVariables [
  User.bIsFrontEnd {type=NiagaraBool}
  User.bIsFrontEndPreview {type=NiagaraBool}
  User.FrontEndScale {type=NiagaraFloat, defaultValue=0.6, description=""}
  User.ScaleLifetime {type=NiagaraFloat, defaultValue=2.5, description=""}
  User.Variant {type=NiagaraInt32}
 ]
}
