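# Niagara system for the DrapeUltra emote's sword fire, a blue IdleFlame ribbon attached to a skeletal mesh.
# A burst spawns two persistent particles, partitioned so each follows its own socket or bone with an offset, spanning the ribbon between them.
# A curve over system age animates a dynamic material parameter.
# Ribbon width scales with the dot product of the system X axis and camera vector, narrowing as they align.
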NiagaraSystem /BRCosmetics/Animation/Game/MainPlayer/Emotes/DrapeUltra/FX/NS_Emote_DrapeUltra_SwordFire
dependencies {
 usedDataInterfaces [&NiagaraDataInterfaceCurve, &NiagaraDataInterfaceCamera, &NiagaraDataInterfaceSkeletalMesh]
 usedDynamicInputs []
 usedModules [
  &self:SystemSpawnScript.SystemScriptSource.SystemScriptGraph.NiagaraNodeAssignment_0.SetVariables_#1
  &SystemState
 ]
}
emitters {IdleFlame {
  emitterUpdateScript [
   &EmitterState(
    Inactive Response=::NewEnumerator1
    Life Cycle Mode=::NewEnumerator1
    Loop Behavior=::NewEnumerator1
    Loop Duration Mode=::NewEnumerator1
    UseLoopDelay=false
   )
   &SpawnBurst_Instantaneous(Spawn Count=2, Use Spawn Probability=false, UseLoopCountLimit=false)
  ]
  particleSpawnScript [
   &PartitionParticles(Particles Per Partition=1, Partition Mode:ENiagara_PartitionMode_v2=::NewEnumerator1) {module=PartitionParticles001}
   &InitializeParticle(
    Color=<rgba 0.043 0.5215 1 1>
    Lifetime=0
    Mesh Scale Mode=::NewEnumerator4
    Ribbon Width=14
    Ribbon Width Mode=::NewEnumerator1
    UsePositionOffset=false
    Write Material Random=true
   )
  ]
  particleUpdateScript [
   &ParticleState(Kill Particles When Lifetime Has Elapsed=false, Let Infinitely Lived Particles Die When Emitter Deactivates=true)
   &SkeletalMeshLocation(
    Bone / Socket Index=$Particles.PartitionParticles001.Partition
    MaskSpawnGroup=false
    Mesh Sampling Type=::NewEnumerator5
    Sampled Position Offset=&MergeHelper_VectorSwitch(0=<0 -5 0>, 1=<-0.5 10 0>, 2=?, 3=?, Index=$Particles.PartitionParticles001.Partition, Num=::NewEnumerator0)
    Skeletal Mesh=$System.SK_Mesh
    Socket Sampling Mode=::NewEnumerator5
   ) {module=SkeletalMeshLocation002}
   /SetVariables() {isSetParameters=true}
   &DynamicMaterialParameters(
    Index 0 Param 1=&Lerp_Float(
     A=-2
     B=0
     Alpha=&FloatFromCurve(
      FloatCurve=DataInterface() {propertyValues {Curve {defaultValue=1, keys [{tangentMode=RCTM_User, time=0.2, value=0.4338}, <at 0.32 0.743>, <at 0.75 0.743>, <at 1 1.5>]}}}
      CurveIndex=$System.Age
     )
    )
    Index 0 Param 2=1
    Module.Index0Float0Write=true
    Module.Index0Float1Write=true
    Module.Index0Float2Write=false
    Module.Index0Float3Write=false
   )
   &ScaleRibbonWidth(Ribbon Width Scale=&Lerp_Float(
     A=1.1
     B=0.5
     Alpha=&Abs_Float(Float:NiagaraFloat=&DotProduct(
       Vector A=/Game/Effects/Niagara/DynamicInputs/Engine/FN_GetSystemAxis(Axis:ENiagaraOrientationAxis=::XAxis, Flip Axis:NiagaraBool=false)
       Vector B=&ReturnCameraVector(Camera Vectors=::NewEnumerator1, Coordinate Space=::Simulation)
      ))
    ))
  ]
  rendererClasses [&NiagaraRibbonRendererProperties]
  renderers [{rendererClass=&NiagaraRibbonRendererProperties}]
 }}
summary {
 systemName=NS_Emote_DrapeUltra_SwordFire
 emitters [{emitterName=IdleFlame, rendererClasses [&NiagaraRibbonRendererProperties]}]
 userVariables [
  User.bIsFrontEnd {type=NiagaraBool}
  User.bIsFrontEndPreview {type=NiagaraBool}
 ]
}
